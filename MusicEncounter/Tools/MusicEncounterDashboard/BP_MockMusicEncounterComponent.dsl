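; Mock music encounter component for the Music Encounter Dashboard tool, routing encounter requests to a mock manager.
; A custom Send Request event calls SendMusicEncounterRequest, whose event implementation here uses the stored Mock Music Encounter Manager when it is available.
; It requests an encounter with the component's instance tag and encounter class, passing Handle Music Encounter Ready as the callback delegate.
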(event Custom|Send%20Request
  (CallFunction|SendMusicEncounterRequest))

(event EventSendMusicEncounterRequest
  (bind _mock_music_encounter_manager (Variables|Default|GetMock%20Music%20Encounter%20Manager)
    (:then
      (MusicEncounterManager|RequestMusicEncounter _mock_music_encounter_manager (Variables|MusicEncounterComponent|GameplayTags|GetMusic%20Encounter%20Instance%20Tag) (Variables|MusicEncounterComponent|GetMusic%20Encounter%20Class) (EventDispatchers|CreateEvent :!event Handle%20Music%20Encounter%20Ready)))
    (:else)))
